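# Physics asset for the BananaPhilosopher back bling mesh, defining collision bodies and joints for a jointed leaf on the pack.
# A large capsule covers the pack, a small one sits at the leaf base, and four identical capsules follow the dyn_leaf bone chain.
# Narrow swing limits and zero-degree twist limits link the chain segments, while the base-to-pack constraint leaves swing and twist free.
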PhysicsAsset /BRCosmetics/Accessories/FORT_Backpacks/Backpack_M_MED_BananaPhilosopher/Meshes/M_MED_BananaPhilosopher_Pack_Physics
bodies {
 base_leaf_jnt_0 [{shapeType=Capsule, length=0.505, radius=0.505, shapeName=base_leaf_jnt_0_capsule, rotation=<rot 0 -90 0>}]
 dyn_leaf_jnt_0 [{shapeType=Capsule, length=2, radius=1, shapeName=dyn_leaf_jnt_0_capsule, center=<4 0 -2>, rotation=<rot 90 0 0>}]
 dyn_leaf_jnt_1 [{shapeType=Capsule, length=2, radius=1, shapeName=dyn_leaf_jnt_1_capsule, center=<4 0 -2>, rotation=<rot 90 0 0>}]
 dyn_leaf_jnt_2 [{shapeType=Capsule, length=2, radius=1, shapeName=dyn_leaf_jnt_2_capsule, center=<4 0 -2>, rotation=<rot 90 0 0>}]
 dyn_leaf_jnt_3 [{shapeType=Capsule, length=2, radius=1, shapeName=dyn_leaf_jnt_3_capsule, center=<4 0 -2>, rotation=<rot 90 0 0>}]
 pack [{
   shapeType=Capsule
   length=17.2814
   radius=9.1839
   shapeName=pack_capsule
   center=<8.3281 3.0862 3.7606>
   rotation=<rot -58.6887 0.0013 0.0529>
  }]
}
constraints [
 {bone1Name=dyn_leaf_jnt_0, bone2Name=base_leaf_jnt_0, swing1LimitDegrees=5, swing2LimitDegrees=10, twistLimitDegrees=0}
 {bone1Name=dyn_leaf_jnt_1, bone2Name=dyn_leaf_jnt_0, swing1LimitDegrees=5, swing2LimitDegrees=15, twistLimitDegrees=0}
 {bone1Name=dyn_leaf_jnt_2, bone2Name=dyn_leaf_jnt_1, swing1LimitDegrees=5, swing2LimitDegrees=15, twistLimitDegrees=0}
 {bone1Name=dyn_leaf_jnt_3, bone2Name=dyn_leaf_jnt_2, swing1LimitDegrees=5, swing2LimitDegrees=15, twistLimitDegrees=0}
 {
  bone1Name=base_leaf_jnt_0
  bone2Name=pack
  swing1LimitDegrees=45
  swing1Motion=Free
  swing2LimitDegrees=45
  swing2Motion=Free
  twistLimitDegrees=45
  twistMotion=Free
 }
]
